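; Parent Blueprint for Battle Royale quest interactable props, revealed only to local players who still need the quest objective.
; After a five-second workaround delay, clients check quest state and size reveal/hide spheres from a curve table, with buffered hide radius.
; Overlaps drive timelines that pop the mesh in, bob and spin it, or shrink it away.
; BlueprintCanInteract gates on objective state; quest updates trigger client-side success effects, fade out and unbind.
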(event EventBeginPlay
  (bind _returnvalue (Fortnite|GetLocalFortPlayerControllers true))
  (bind _output (Utilities|Array|Get(acopy) _returnvalue))
  (bind _questitem (Variables|QuestStuff|GetQuestItem))
  ;[1 0.5445 0.1188] Workaround - needs to be something better for ship
  (Utilities|FlowControl|Delay 5.0)
  ;[1 1 1] CLIENT AND SERVER - Ensure playlist is loaded
  (bind _asfort_game_state_athena (Utilities|Casting|CastToFortGameStateAthena (Game|GetGameState)))
  (if (not (Variables|/Script/FortniteGame.FortGameStateAthena|GetbPlaylistDataIsLoaded _asfort_game_state_athena))
    (Playlist|BindEventtoOnCurrentPlaylistLoaded _asfort_game_state_athena (delegate Custom|PlaylistLoaded)))
  (Networking|SwitchHasAuthority
    (:Remote
      (Utilities|FlowControl|Sequence
        (:then_0
          ;[1 1 1] Hide or Unhide Object depending on if player has quest or not
          (if (Quest|CanProgressQuestinCurrentZone (Quests|GetQuestManager _output "Invalid") _questitem)
            (bind (_questisvalid _questcompleted _objectivecompleted) (Default|AthenaCheckQuestandObjectiveCompletion _output _questitem (Variables|QuestStuff|GetQuestBackendObjectiveName)))
            (if (not _objectivecompleted)
              (Variables|Default|SetShouldBeHidden false)
              (Variables|Default|SetisLocallyValid true)
              ;[1 1 1] Bind Event for When Quest Gets Updated
              (bind _asfort_player_controller_athena (Utilities|Casting|CastToFortPlayerControllerAthena (Utilities|Array|Get(acopy) (Fortnite|GetLocalFortPlayerControllers true))))
              (Quests|BindEventtoOnQuestsUpdated (Quests|GetQuestManager _asfort_player_controller_athena "Invalid") (EventDispatchers|CreateEvent :!event Custom|BattlestarQuestUpdated)))
            (else
              _))
          (Variables|Default|SetShouldBeHidden true)
          (Variables|Default|SetisLocallyValid false))
        (:then_1
          ;[1 1 1] Initialize Object location and set Hide/Unhide distance
          (Variables|Default|SetSMLocationInit (Transformation|GetWorldLocation (Variables|Default|GetSM_Quest_Object)))
          (bind _outxy (DataTable|EvaluateCurveTableRow "/Game/Athena/Balance/DataTables/AthenaGameData.AthenaGameData" "Default.QuestIneractable.RevealRadius")
            (:RowFound
              (Variables|Default|SetRevealRadius _outxy)
              (Components|Sphere|SetSphereRadius (Variables|Default|GetSphereCollisionReveal) (Variables|Default|GetRevealRadius) true)
              (Components|Sphere|SetSphereRadius (Variables|Default|GetSphereCollisionHide) (+ (Variables|Default|GetRevealRadius) (Variables|Default|GetRevealHideRadiusBuffer)) true))
            (:RowNotFound
              _))
          ;[1 1 1] Only Proceed if Quest is active
          (if (not (Variables|Default|GetShouldBeHidden))
            (Components|Activation|Activate (Variables|Default|GetQuestObjectAmbientLoopSound) false)
            (Utilities|FlowControl|Sequence
              (:then_0
                ;[1 1 1] Pop Object into view
                (Collision|BindEventtoOnComponentBeginOverlap (Variables|Default|GetSphereCollisionReveal) (delegate Custom|ShowStar)))
              (:then_1
                ;[1 1 1] Hide Object when outside of view range.
                (Collision|BindEventtoOnComponentEndOverlap (Variables|Default|GetSphereCollisionHide) (delegate Custom|HideStar))))))))))

(event Custom|PlaylistLoaded
  (params (PlaylistName :type Name) (PlaylistContextTags :type Gameplay%20Tag%20Container%20Structure%20%28by%20ref%29)))

(event Custom|ShowStar
  (params (OverlappedComponent :type Primitive%20Component%20Object%20Reference) (OtherActor :type Actor%20Object%20Reference) (OtherComp :type Primitive%20Component%20Object%20Reference) (OtherBodyIndex :type Integer) (bFromSweep :type Boolean) (SweepResult :type Hit%20Result%20Structure%20%28by%20ref%29))
  ;[1 1 1] Pop Object into view
  (Utilities|Casting|CastToFortPawn OtherActor)
  (if (not (Variables|Default|GetShouldBeHidden))
    (Rendering|Components|Light|SetIntensity (Variables|Default|GetPointLight) 2e+03)
    (Development|SetHiddeninGame (Variables|Default|GetSM_Quest_Object) false false)
    (bind (_direction _revealtime) (Timeline|Reveal PlayFromStart)
      (:Update
        (Transformation|SetWorldScale3D (Variables|Default|GetSM_Quest_Object) (Math|Vector|Lerp(Vector) "0.010000,0.010000,0.010000" "1.000000,1.000000,1.000000" _revealtime)))
      (:Finished
        ;[1 1 1] Make Object Bob and Spin
        (Components|Timeline|SetLooping (Variables|Prop_QuestObject_Interactable_Parent|GetBobAndSpin) true)
        (bind (_direction_2 _bob _spin) (Timeline|BobAndSpin Play)
          (:Update
            (Transformation|SetWorldLocationAndRotation (Variables|Default|GetSM_Quest_Object) (Math|Vector|Lerp(Vector) (Variables|Default|GetSMLocationInit) (+ (Variables|Default|GetSMLocationInit) "0, 0,20.000000") _bob) (Math|Rotator|Lerp(Rotator) "0, 0, 0" "0,720.000000, 0" _spin) false false)))))))

(event Custom|HideStar
  (params (OverlappedComponent :type Primitive%20Component%20Object%20Reference) (OtherActor :type Actor%20Object%20Reference) (OtherComp :type Primitive%20Component%20Object%20Reference) (OtherBodyIndex :type Integer))
  ;[1 1 1] Hide Object when outside of view range.
  (Utilities|Casting|CastToFortPawn OtherActor)
  (if (not (Variables|Default|GetShouldBeHidden))
    (Utilities|FlowControl|Sequence
      (:then_0
        ;[1 1 1] Pop Object into view
        (Timeline|Reveal Stop))
      (:then_1
        _)
      (:then_2
        (bind (_direction _scaledowntime) (Timeline|Hide PlayFromStart)
          (:Update
            (Transformation|SetWorldScale3D (Variables|Default|GetSM_Quest_Object) (Math|Vector|Lerp(Vector) "1.000000,1.000000,1.000000" "0.010000,0.010000,0.010000" _scaledowntime)))
          (:Finished
            (Development|SetHiddeninGame (Variables|Default|GetSM_Quest_Object) true false)
            (Rendering|Components|Light|SetIntensity (Variables|Default|GetPointLight) 0.0)))))
    (Timeline|BobAndSpin Stop)))

(event Custom|BattlestarQuestUpdated
  ;[1 1 1] Quest Updated
  (if (Variables|Default|GetisLocallyValid)
    (bind (_questisvalid _questcompleted _objectivecompleted) (Default|AthenaCheckQuestandObjectiveCompletion (Utilities|Array|Get(acopy) (Fortnite|GetLocalFortPlayerControllers true)) (Variables|QuestStuff|GetQuestItem) (Variables|QuestStuff|GetQuestBackendObjectiveName)))
    (if _questcompleted
      _
      (elif _objectivecompleted
        _))
    (CallFunction|PlaySuccessFXandSound)
    (Variables|Default|SetisLocallyValid false)
    (Variables|Default|SetShouldBeHidden true)
    ;[1 1 1] Unbind Quest Updated Event
    (bind _asfort_player_controller_athena (Utilities|Casting|CastToFortPlayerControllerAthena (Utilities|Array|Get(acopy) (Fortnite|GetLocalFortPlayerControllers true))))
    (Quests|UnbindEventfromOnQuestsUpdated (Quests|GetQuestManager _asfort_player_controller_athena "Invalid") (EventDispatchers|CreateEvent :!event Custom|BattlestarQuestUpdated))))

(event Custom|PlaySuccessFXandSound
  (bind _returnvalue (Transformation|GetActorLocation))
  ;[1 1 1] Play Success FX for local player and hide object
  (Networking|SwitchHasAuthority
    (:Remote
      (Audio|SpawnSoundatLocation (Variables|Default|GetSuccessSound) _returnvalue)
      (Effects|SpawnEmitteratLocation (Variables|Default|GetSuccessEmitter) _returnvalue)
      (Rendering|Components|Light|SetIntensity (Variables|Default|GetPointLight) 0.0)
      (Audio|Components|Audio|Stop (Variables|Default|GetQuestObjectAmbientLoopSound))
      (Game|Feedback|ClientPlayForceFeedback (Utilities|Array|Get(acopy) (Fortnite|GetLocalFortPlayerControllers true)) (Variables|Default|GetQuestSuccessFF) "None" false false false)
      (bind _returnvalue_2 (Rendering|Material|CreateDynamicMaterialInstance (Variables|Default|GetSM_Quest_Object) 0 0 "None" :!class /Script/Engine.PrimitiveComponent))
      (Variables|Default|SetGlowyYellowMID _returnvalue_2)
      (bind (_direction _fadetime) (Timeline|FadeOut PlayFromStart)
        (:Update
          (Rendering|Material|SetScalarParameterValue (Variables|Default|GetGlowyYellowMID) "Opacity" _fadetime :!class /Script/Engine.MaterialInstanceDynamic))
        (:Finished
          (Rendering|SetActorHiddenInGame true))))))

(event EventTick
  (params (DeltaSeconds :type Float%20%28single-precision%29)))

(fn BlueprintCanInteract
  (params (InteractingPawn :type Fort%20Pawn%20Object%20Reference) (InteractionBeingAttempted :type EInteractionBeingAttempted%20Enum :default "FirstInteraction") (InteractionType :type TInteractionType%20Enum :default "IT_NoInteraction"))
  (Utilities|IsValid InteractingPawn
    (:Is%20Valid
      (bind _asfort_player_controller_athena (Utilities|Casting|CastToFortPlayerControllerAthena (Game|GetInstigatorController InteractingPawn)))
      (if (Quest|CanProgressQuestinCurrentZone (Quests|GetQuestManager _asfort_player_controller_athena "Invalid") (Variables|QuestStuff|GetQuestItem))
        (bind (_questisvalid _questcompleted _objectivecompleted) (Default|AthenaCheckQuestandObjectiveCompletion _asfort_player_controller_athena (Variables|QuestStuff|GetQuestItem) (Variables|QuestStuff|GetQuestBackendObjectiveName)))
        (if (and _questisvalid (not _objectivecompleted))
          (LocalVariables|SetQuestValid true)
          (else
            (LocalVariables|SetQuestValid false)))
        (return (LocalVariables|GetQuestValid))
        (else
          _)))))

(fn BlueprintGetInteractionString
  (params (InteractingPawn :type Fort%20Pawn%20Object%20Reference) (InteractionBeingAttempted :type EInteractionBeingAttempted%20Enum :default "FirstInteraction"))
  (return (Variables|QuestStuff|GetQuestInteractString)))

(fn OnRep_AllPlayerAccountWhoHaveInteracted)

(fn ConstructionScript
  (bind _returnvalue (Rendering|Material|CreateDynamicMaterialInstance (Variables|Default|GetSM_Quest_Object) (Variables|Default|GetObjectMatElementIndex) (Variables|Default|GetObjectSourceMaterial) "None" :!class /Script/Engine.PrimitiveComponent))
  (Variables|Default|SetGlowyYellowMID _returnvalue))
